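# Translucent UI material for the backing of the quest top bar in the minimap's quest rewards panel.
# It reads the Slate post buffer and blends it toward a half-brightness copy of itself along a rotated linear gradient.
# The same gradient also feeds an ease-curve function, whose output no other node uses.
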Material /FortniteCoreUI/HUD/MiniMap/QuestRewardsPanel/Materials/M_UI_TopBar_Quest_Backing
graph {
 properties {
  bUsedWithStaticMesh=false
  blendMode=BLEND_Translucent
  materialDomain=MD_UI
  pixelDepthOffsetMode=PDOM_Legacy
  usedShadingModels=Default Lit
 }
 nodes [
  n1 = MaterialFunctionCall!("UVs (Optional)") {materialFunction=/Engine/Functions/UserInterface/GetSlatePost0}
  n2 = MaterialFunctionCall(GradientStart=Constant(), GradientEnd=Constant() {r=1}, RotationAngle=Constant() {r=0.25}) {materialFunction=&MF_UI_LinearGradient}
  LinearInterpolate(A=@n1.LinearRGB, B=Multiply(A=@n1.LinearRGB) {constB=0.5}, Alpha=@n2.Result)
  MaterialFunctionCall("Value (0-1)"=@n2.Result) {materialFunction=&MF_UI_EaseCurves}
 ]
}
