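# NPC quest rewards panel widget, based on AthenaNPCQuestRewardsPanel, showing an NPC's name, character picture and a rewards blurb over glow backgrounds.
# Show sets the name and lazily loaded picture, and uses bIsLocked to toggle the lock badge and pick Battle Pass or Snapshot Questline text.
# Loaded widget component classes are created and added to fill the Container overlay.
# The portrait's dynamic material gets its clipping mask zeroed on PreConstruct and Show.
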WidgetBlueprint /FortniteCoreUI/HUD/MiniMap/QuestRewardsPanel/W_NPC_QuestRewardsPanel_RewardsMode
graphs [EventGraph {nodes [
   n1 = Event BP_OnWidgetComponentClassLoaded!() -> (OutputDelegate, then, WidgetComponentId:Name=None, WidgetClass:User Widget Class Reference) {is=title K2Node_Event}
   > n2 = Create Widget!(execute:Exec, Class:User Widget Class Reference=@n1.WidgetClass, OwningPlayer:Player Controller Object Reference=GetOwningPlayer()) -> (then:Exec, ReturnValue:User Widget Object Reference)
   > n3 = AddChildToOverlay!(execute, self:Overlay Object Reference=Get Container() -> (Container:Overlay Object Reference), Content:Widget Object Reference=@n2) -> (then, ReturnValue:Overlay Slot Object Reference)
   > SetHorizontalAlignment(self=@n3, InHorizontalAlignment=HAlign_Fill)
   > SetVerticalAlignment(self=@n3, InVerticalAlignment=VAlign_Fill)
   Event PreConstruct()
   > n4 = GetDynamicMaterial(self=Get LazyImg_Character() -> (LazyImg_Character:Fort Lazy Image Object Reference))
   > SetScalarParameterValue(self=@n4, ParameterName=ClippingMask, Value=0)
   > SetScalarParameterValue(self=@n4, ParameterName=Disable512Scaling, Value=1)
   n5 = Event Show!() -> (OutputDelegate, then, NPCTitle:"Text (by ref)", NPCPicture:"Soft Object Reference (by ref)", bIsLocked:bool) {is=title K2Node_Event}
   > n6 = Sequence!(execute) -> (then_0, then_1, then_2:Exec, then_3:Exec, then_4:Exec)
   > Set Text(
    Text=@n5.NPCTitle
    self:Rich Text Block Object Reference=Get RichTextBlock_NPCName() -> (RichTextBlock_NPCName:Common Rich Text Block Object Reference)
   )
   Set Text(
    execute=@n6.then_2
    Text=Select(
     Option 0:Text=Format Text!(Format="LOCTEXT(\"Play through the {0} Snapshot Questline to unlock these extra cosmetics!\")", 0:"Text (by ref)"=@n5.NPCTitle) -> (Result)
     Option 1:Text="LOCTEXT(\"<Highlight>Purchase the Battle Pass</> to be able to unlock these rewards.\")"
     Index=@n5.bIsLocked
    ) -> (ReturnValue:Text)
    self=Get RichText_Description() -> (RichText_Description:Common Rich Text Block Object Reference)
   )
   n7 = Get LazyImg_Character()
   SetBrushFromLazyDisplayAsset!(
    execute=@n6.then_3
    self:Common Lazy Image Object Reference=@n7
    LazyObject:"Soft Object Reference (by ref)"=@n5.NPCPicture
    bMatchTextureSize:bool=false
   ) -> (then)
   n8 = GetDynamicMaterial(execute=@n6.then_4, self=@n7)
   > SetScalarParameterValue(self=@n8, ParameterName=ClippingMask, Value=0)
   Set Visibility(
    execute=@n6.then_1
    Visibility=Select(Option 0:ESlateVisibility Enum=Collapsed, Option 1:ESlateVisibility Enum=SelfHitTestInvisible, Index=@n5.bIsLocked) -> (ReturnValue:ESlateVisibility Enum)
    self=Get LockStuff() -> (LockStuff:Overlay Object Reference)
   )
  ]}]
variables []
widgets {
 info {inheritedWidgetCount=0, parentClass=/Script/FortniteUI.AthenaNPCQuestRewardsPanel, rootWidgetClass=&SizeBox, widgetCount=13}
 widgets [
  SizeBox_0 {widgetClassPath=&SizeBox, widget=SizeBox_0, properties {bOverride_WidthOverride=true, heightOverride=940, widthOverride=847}}
  Container {widgetClassPath=&Overlay, bIsVariable=true, parent=SizeBox_0, widget=Container, slot {class=SizeBoxSlot}}
  HorizontalBox_0 {
   widgetClassPath=&HorizontalBox
   parent=Container
   widget=HorizontalBox_0
   slot {class=OverlaySlot, properties {horizontalAlignment=HAlign_Center, padding=<ltrb 0 80 0 0>}}
  }
  LockStuff {
   widgetClassPath=&Overlay
   bIsVariable=true
   parent=HorizontalBox_0
   widget=LockStuff
   properties {renderTransform {angle=0, scale=<1 1>, shear=<0 0>, translation=<0 -4>}, renderTransformPivot=<0 1>}
   slot {class=HorizontalBoxSlot, properties {horizontalAlignment=HAlign_Center, verticalAlignment=VAlign_Center, padding=<ltrb 0 0 7 0>}}
  }
  Shadow {
   widgetClassPath=&Image
   bIsVariable=true
   parent=LockStuff
   widget=Shadow
   properties {
    brush {
     bIsDynamicallyLoaded=false
     drawAs=Image
     imageType=NoImage
     mirroring=NoMirror
     resourceName=None
     resourceObject=null
     tiling=NoTile
     imageSize=<48 48>
     margin=<ltrb 0 0 0 0>
     outlineSettings {
      bUseBrushTransparency=false
      roundingType=HalfHeightRadius
      width=0
      color {colorUseRule=UseColor_Specified, specifiedColor=<rgba 0 0 0 0>}
      cornerRadii=<0 0 0 0>
     }
     tintColor {colorUseRule=UseColor_Specified, specifiedColor=<rgba 1 1 1 1>}
     uVRegion {bIsValid=false, max=<0 0>, min=<0 0>}
    }
    colorAndOpacity=<rgba 0.0549 0.0745 0.3294 1>
    renderTransform {angle=0, scale=<1 1>, shear=<-9 0>, translation=<1.5 1.5>}
   }
   slot {class=OverlaySlot, properties {horizontalAlignment=HAlign_Fill, verticalAlignment=VAlign_Fill}}
  }
  Image_305 {
   widgetClassPath=&Image
   bIsVariable=true
   parent=LockStuff
   widget=Image_305
   properties {
    brush {
     bIsDynamicallyLoaded=false
     drawAs=Image
     imageType=NoImage
     mirroring=NoMirror
     resourceName=None
     resourceObject=null
     tiling=NoTile
     imageSize=<48 48>
     margin=<ltrb 0 0 0 0>
     outlineSettings {
      bUseBrushTransparency=false
      roundingType=HalfHeightRadius
      width=0
      color {colorUseRule=UseColor_Specified, specifiedColor=<rgba 0 0 0 0>}
      cornerRadii=<0 0 0 0>
     }
     tintColor {colorUseRule=UseColor_Specified, specifiedColor=<rgba 1 1 1 1>}
     uVRegion {bIsValid=false, max=<0 0>, min=<0 0>}
    }
    colorAndOpacity=<rgba 1 0 0.2988 1>
    renderTransform {angle=0, scale=<1 1>, shear=<-9 0>, translation=<0 0>}
   }
   slot {class=OverlaySlot, properties {horizontalAlignment=HAlign_Fill, verticalAlignment=VAlign_Fill}}
  }
  ImgLockIcon {
   widgetClassPath=&Image
   bIsVariable=true
   parent=LockStuff
   widget=ImgLockIcon
   properties {brush {
     bIsDynamicallyLoaded=false
     drawAs=Image
     imageType=NoImage
     mirroring=NoMirror
     resourceName=None
     resourceObject=/Game/UI/Foundation/Textures/Icons/Locks/T-Icon-Lock-64
     tiling=NoTile
     imageSize=<32 32>
     margin=<ltrb 0 0 0 0>
     outlineSettings {
      bUseBrushTransparency=false
      roundingType=HalfHeightRadius
      width=0
      color {colorUseRule=UseColor_Specified, specifiedColor=<rgba 0 0 0 0>}
      cornerRadii=<0 0 0 0>
     }
     tintColor {colorUseRule=UseColor_Specified, specifiedColor=<rgba 1 1 1 1>}
     uVRegion {bIsValid=false, max=<0 0>, min=<0 0>}
    }}
   slot {class=OverlaySlot, properties {horizontalAlignment=HAlign_Center, verticalAlignment=VAlign_Center}}
  }
  RichTextBlock_NPCName {
   widgetClassPath=&CommonRichTextBlock
   bIsVariable=true
   parent=HorizontalBox_0
   widget=RichTextBlock_NPCName
   properties {
    bOverrideDefaultStyle=true
    justification=Center
    lineHeightPercentage=0.9
    scrollStyle=&TextScrollStyle-Base_C
    text=NPC NAME
    textTransformPolicy=ToUpper
    defaultTextStyleOverride {
     overflowPolicy=Clip
     transformPolicy=None
     colorAndOpacity {colorUseRule=UseColor_Specified, specifiedColor=<rgba 1 1 1 1>}
     font {
      bForceMonospaced=false
      bMaterialIsStencil=false
      fontMaterial=null
      fontObject=&Burbank
      letterSpacing=0
      monospacedWidth=1
      size=54
      skewAmount=0.1
      typefaceFontName=BurbankRegular-Black
      outlineSettings {
       bApplyOutlineToDropShadows=false
       bMiteredCorners=false
       bSeparateFillAlpha=false
       outlineMaterial=null
       outlineSize=0
       outlineColor=<rgba 0 0 0 1>
      }
     }
     highlightColor {colorUseRule=UseColor_Specified, specifiedColor=<rgba 0 0 0 1>}
     highlightShape {
      bIsDynamicallyLoaded=false
      drawAs=Image
      imageType=NoImage
      mirroring=NoMirror
      resourceName=None
      resourceObject=null
      tiling=NoTile
      imageSize=<32 32>
      margin=<ltrb 0 0 0 0>
      outlineSettings {
       bUseBrushTransparency=false
       roundingType=HalfHeightRadius
       width=0
       color {colorUseRule=UseColor_Specified, specifiedColor=<rgba 0 0 0 0>}
       cornerRadii=<0 0 0 0>
      }
      tintColor {colorUseRule=UseColor_Specified, specifiedColor=<rgba 1 1 1 1>}
      uVRegion {bIsValid=false, max=<0 0>, min=<0 0>}
     }
     selectedBackgroundColor {colorUseRule=UseColor_Foreground, specifiedColor=<rgba 1 0 1 1>}
     shadowColorAndOpacity=<rgba 0.0549 0.0745 0.3294 1>
     shadowOffset=<2 2>
     strikeBrush {
      bIsDynamicallyLoaded=false
      drawAs=NoDrawType
      imageType=NoImage
      mirroring=NoMirror
      resourceName=None
      resourceObject=null
      tiling=NoTile
      imageSize=<32 32>
      margin=<ltrb 0 0 0 0>
      outlineSettings {
       bUseBrushTransparency=false
       roundingType=HalfHeightRadius
       width=0
       color {colorUseRule=UseColor_Specified, specifiedColor=<rgba 0 0 0 0>}
       cornerRadii=<0 0 0 0>
      }
      tintColor {colorUseRule=UseColor_Specified, specifiedColor=<rgba 1 1 1 1>}
      uVRegion {bIsValid=false, max=<0 0>, min=<0 0>}
     }
     underlineBrush {
      bIsDynamicallyLoaded=false
      drawAs=Image
      imageType=NoImage
      mirroring=NoMirror
      resourceName=None
      resourceObject=null
      tiling=NoTile
      imageSize=<32 32>
      margin=<ltrb 0 0 0 0>
      outlineSettings {
       bUseBrushTransparency=false
       roundingType=HalfHeightRadius
       width=0
       color {colorUseRule=UseColor_Specified, specifiedColor=<rgba 0 0 0 0>}
       cornerRadii=<0 0 0 0>
      }
      tintColor {colorUseRule=UseColor_Specified, specifiedColor=<rgba 1 1 1 1>}
      uVRegion {bIsValid=false, max=<0 0>, min=<0 0>}
     }
    }
    margin=<ltrb 0 0 0 16>
   }
   slot {class=HorizontalBoxSlot}
  }
  Image_BackgroundGlow {
   widgetClassPath=&Image
   bIsVariable=true
   parent=Container
   widget=Image_BackgroundGlow
   properties {
    brush {
     bIsDynamicallyLoaded=false
     drawAs=Image
     imageType=NoImage
     mirroring=NoMirror
     resourceName=None
     resourceObject=/FortniteCoreUI/HUD/MiniMap/QuestRewardsPanel/Materials/MI_UI_NPCQuestRewards_Glow_BG
     tiling=NoTile
     imageSize=<1025 1025>
     margin=<ltrb 0 0 0 0>
     outlineSettings {
      bUseBrushTransparency=false
      roundingType=HalfHeightRadius
      width=0
      color {colorUseRule=UseColor_Specified, specifiedColor=<rgba 0 0 0 0>}
      cornerRadii=<0 0 0 0>
     }
     tintColor {colorUseRule=UseColor_Specified, specifiedColor=<rgba 1 1 1 1>}
     uVRegion {bIsValid=false, max=<0 0>, min=<0 0>}
    }
    colorAndOpacity=<rgba 1 1 1 0.6714>
    renderTransform {angle=0, scale=<1.3 1.3>, shear=<0 0>, translation=<0 -100>}
   }
   slot {class=OverlaySlot, properties {horizontalAlignment=HAlign_Center, verticalAlignment=VAlign_Center}}
  }
  Image_ForegroundGlow {
   widgetClassPath=&Image
   bIsVariable=true
   parent=Container
   widget=Image_ForegroundGlow
   properties {
    brush {
     bIsDynamicallyLoaded=false
     drawAs=Image
     imageType=NoImage
     mirroring=NoMirror
     resourceName=None
     resourceObject=&MI_UI_NPCQuestRewards_Glow_BG
     tiling=NoTile
     imageSize=<1024 1024>
     margin=<ltrb 0 0 0 0>
     outlineSettings {
      bUseBrushTransparency=false
      roundingType=HalfHeightRadius
      width=0
      color {colorUseRule=UseColor_Specified, specifiedColor=<rgba 0 0 0 0>}
      cornerRadii=<0 0 0 0>
     }
     tintColor {colorUseRule=UseColor_Specified, specifiedColor=<rgba 1 1 1 1>}
     uVRegion {bIsValid=false, max=<0 0>, min=<0 0>}
    }
    colorAndOpacity=<rgba 1 1 1 0.0417>
    renderTransform {angle=0, scale=<1.25 1.25>, shear=<0 0>, translation=<0 -83>}
   }
   slot {class=OverlaySlot, properties {horizontalAlignment=HAlign_Right, verticalAlignment=VAlign_Center, padding=<ltrb 0 0 -50 -200>}}
  }
  LazyImg_Character {
   widgetClassPath=&FortLazyImage
   bIsVariable=true
   parent=Container
   widget=LazyImg_Character
   properties {
    materialTextureParamName=ItemIcon
    brush {
     bIsDynamicallyLoaded=false
     drawAs=Image
     imageType=NoImage
     mirroring=NoMirror
     resourceName=None
     resourceObject=/FortniteCoreUI/HUD/MiniMap/QuestRewardsPanel/Materials/MI_UI_Character_Picture
     tiling=NoTile
     imageSize=<500 500>
     margin=<ltrb 0 0 0 0>
     outlineSettings {
      bUseBrushTransparency=false
      roundingType=HalfHeightRadius
      width=0
      color {colorUseRule=UseColor_Specified, specifiedColor=<rgba 0 0 0 0>}
      cornerRadii=<0 0 0 0>
     }
     tintColor {colorUseRule=UseColor_Specified, specifiedColor=<rgba 1 1 1 1>}
     uVRegion {bIsValid=false, max=<0 0>, min=<0 0>}
    }
    renderTransform {angle=0, scale=<1.7 1.7>, shear=<0 0>, translation=<6 100>}
   }
   slot {class=OverlaySlot, properties {horizontalAlignment=HAlign_Center, verticalAlignment=VAlign_Center}}
  }
  Border_53 {
   widgetClassPath=&Border
   parent=Container
   widget=Border_53
   properties {
    background {
     bIsDynamicallyLoaded=false
     drawAs=Image
     imageType=NoImage
     mirroring=NoMirror
     resourceName=None
     resourceObject=/FortniteCoreUI/HUD/MiniMap/QuestRewardsPanel/Materials/MI_UI_QR_InfoBackground
     tiling=NoTile
     imageSize=<32 32>
     margin=<ltrb 0 0 0 0>
     outlineSettings {
      bUseBrushTransparency=false
      roundingType=HalfHeightRadius
      width=0
      color {colorUseRule=UseColor_Specified, specifiedColor=<rgba 0 0 0 0>}
      cornerRadii=<0 0 0 0>
     }
     tintColor {colorUseRule=UseColor_Specified, specifiedColor=<rgba 1 1 1 1>}
     uVRegion {bIsValid=false, max=<0 0>, min=<0 0>}
    }
    padding=<ltrb 50 20 50 20>
   }
   slot {class=OverlaySlot, properties {horizontalAlignment=HAlign_Center, verticalAlignment=VAlign_Bottom, padding=<ltrb 0 0 0 30>}}
  }
  RichText_Description {
   widgetClassPath=&CommonRichTextBlock
   bIsVariable=true
   parent=Border_53
   widget=RichText_Description
   properties {
    autoWrapText=true
    justification=Center
    text=Description
    textStyleSet=/FortniteCoreUI/HUD/MiniMap/QuestRewardsPanel/Materials/DT_UI_RichTextStyles_QuestRewards
    wrapTextAt=800
    decoratorClasses [/Game/UI/Sidebar/RebootRally/BP_UI_RebootRallyRichImageDecorator.BP_UI_RebootRallyRichImageDecorator_C]
    defaultTextStyleOverride {
     overflowPolicy=Clip
     transformPolicy=None
     colorAndOpacity {colorUseRule=UseColor_Specified, specifiedColor=<rgba 1 1 1 1>}
     font {
      bForceMonospaced=false
      bMaterialIsStencil=false
      fontMaterial=null
      fontObject=&BurbankSmall
      letterSpacing=0
      monospacedWidth=1
      size=19
      skewAmount=0
      typefaceFontName=Bold
      outlineSettings {
       bApplyOutlineToDropShadows=false
       bMiteredCorners=false
       bSeparateFillAlpha=false
       outlineMaterial=null
       outlineSize=0
       outlineColor=<rgba 0 0 0 1>
      }
     }
     highlightColor {colorUseRule=UseColor_Specified, specifiedColor=<rgba 0 0 0 1>}
     highlightShape {
      bIsDynamicallyLoaded=false
      drawAs=Image
      imageType=NoImage
      mirroring=NoMirror
      resourceName=None
      resourceObject=null
      tiling=NoTile
      imageSize=<32 32>
      margin=<ltrb 0 0 0 0>
      outlineSettings {
       bUseBrushTransparency=false
       roundingType=HalfHeightRadius
       width=0
       color {colorUseRule=UseColor_Specified, specifiedColor=<rgba 0 0 0 0>}
       cornerRadii=<0 0 0 0>
      }
      tintColor {colorUseRule=UseColor_Specified, specifiedColor=<rgba 1 1 1 1>}
      uVRegion {bIsValid=false, max=<0 0>, min=<0 0>}
     }
     selectedBackgroundColor {colorUseRule=UseColor_Foreground, specifiedColor=<rgba 1 0 1 1>}
     shadowColorAndOpacity=<rgba 0 0 0 1>
     shadowOffset=<0 0>
     strikeBrush {
      bIsDynamicallyLoaded=false
      drawAs=Image
      imageType=NoImage
      mirroring=NoMirror
      resourceName=None
      resourceObject=null
      tiling=NoTile
      imageSize=<32 32>
      margin=<ltrb 0 0 0 0>
      outlineSettings {
       bUseBrushTransparency=false
       roundingType=HalfHeightRadius
       width=0
       color {colorUseRule=UseColor_Specified, specifiedColor=<rgba 0 0 0 0>}
       cornerRadii=<0 0 0 0>
      }
      tintColor {colorUseRule=UseColor_Specified, specifiedColor=<rgba 1 1 1 1>}
      uVRegion {bIsValid=false, max=<0 0>, min=<0 0>}
     }
     underlineBrush {
      bIsDynamicallyLoaded=false
      drawAs=Image
      imageType=NoImage
      mirroring=NoMirror
      resourceName=None
      resourceObject=null
      tiling=NoTile
      imageSize=<32 32>
      margin=<ltrb 0 0 0 0>
      outlineSettings {
       bUseBrushTransparency=false
       roundingType=HalfHeightRadius
       width=0
       color {colorUseRule=UseColor_Specified, specifiedColor=<rgba 0 0 0 0>}
       cornerRadii=<0 0 0 0>
      }
      tintColor {colorUseRule=UseColor_Specified, specifiedColor=<rgba 1 1 1 1>}
      uVRegion {bIsValid=false, max=<0 0>, min=<0 0>}
     }
    }
   }
   slot {class=BorderSlot, properties {padding=<ltrb 50 20 50 20>}}
  }
 ]
}
