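# Minimap HUD widget for a collection NPC's services, showing an optional "Character Rep" header above an inherited list of service-group entries.
# PopulateServiceData collapses the header when progress is off; otherwise EnableHeader sets its visibility and the current/max benefit levels fill the rich-text count.
# The level ratio drives the bar material's AnimateProgress parameter, substituting a slightly negative value when the ratio is zero or less.
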WidgetBlueprint /FortniteCoreUI/HUD/MiniMap/Services/CollectionNPCServiceContainer
graphs [
 EventGraph {nodes [
   n1 = Event PopulateServiceData!() -> (OutputDelegate, then, bShowProgress:bool, CurrentBenefitsLvl:int=0, TotalBenefitsLvl:int=0) {is=title K2Node_Event}
   > n2 = Branch(Condition=@n1.bShowProgress)
   > PopulateHeaderRepLevel!(execute, self, Current:int=@n1.CurrentBenefitsLvl, Max:int=@n1.TotalBenefitsLvl) -> (then)
   PopulateHeaderNormal(execute=@n2.else)
  ]}
 PopulateHeaderNormal {nodes [
   PopulateHeaderNormal() {is=title K2Node_FunctionEntry}
   > SetVisibility(self=Get VB_RepLevel() -> (VB_RepLevel:Vertical Box Object Reference), InVisibility=Collapsed)
  ]}
 PopulateHeaderRepLevel {nodes [
   n1 = PopulateHeaderRepLevel!() -> (then, Current:int, Max:int) {is=title K2Node_FunctionEntry}
   > SetVisibility(
    self=Get VB_RepLevel()
    InVisibility=Select(Option 0=Collapsed, Option 1=SelfHitTestInvisible, Index=Get EnableHeader() -> (EnableHeader:bool))
   )
   > SetText(
    self=Get RichText_Progress() -> (RichText_Progress:Common Rich Text Block Object Reference)
    InText=Format Text!(Format="LOCTEXT(\"<text color=\\\"FFF700\\\">{Current}</> / {Max}\")", Current:int=@n1, Max:int=@n1.Max) -> (Result)
   )
   > n2 = GetDynamicMaterial(self=Get Progress() -> (Progress:Image Object Reference))
   n3 = float / float(A="To Float (Integer)"(InInt=@n1), B="To Float (Integer)"(InInt=@n1.Max))
   > SetScalarParameterValue(
    self=@n2
    ParameterName=AnimateProgress
    Value=Select(Option 0:float=@n3, Option 1:float=-0.1, Index="float <= float"(A=@n3) -> (ReturnValue=~)) -> (ReturnValue:float)
   )
  ]}
]
variables [EnableHeader {type=boolean, default=true}]
widgets {
 info {inheritedWidgetCount=1, parentClass=/Script/FortniteUI.CollectionNPCServiceContainer, rootWidgetClass=&Overlay, widgetCount=9}
 widgets [
  Overlay_0 {widgetClassPath=&Overlay, widget=Overlay_0}
  Background {
   widgetClassPath=&Image
   parent=Overlay_0
   widget=Background
   properties {renderOpacity=0.5, colorAndOpacity=<rgba 0 0 0 1>}
   slot {class=OverlaySlot, properties {horizontalAlignment=HAlign_Fill, verticalAlignment=VAlign_Fill}}
  }
  VerticalBox_17 {
   widgetClassPath=&VerticalBox
   parent=Overlay_0
   widget=VerticalBox_17
   slot {class=OverlaySlot, properties {padding=<ltrb 20 10 20 10>}}
  }
  VB_RepLevel {widgetClassPath=&VerticalBox, bIsVariable=true, parent=VerticalBox_17, widget=VB_RepLevel, slot {class=VerticalBoxSlot}}
  Text_RepHeader {
   widgetClassPath=&CommonTextBlock
   parent=VB_RepLevel
   widget=Text_RepHeader
   properties {
    mobileFontSizeMultiplier=1.3
    text=Character Rep:
    font {
     bForceMonospaced=false
     bMaterialIsStencil=false
     fontMaterial=null
     fontObject=&HeadingNow
     letterSpacing=0
     monospacedWidth=1
     size=12.75
     skewAmount=0
     typefaceFontName=75 Medium
     outlineSettings {
      bApplyOutlineToDropShadows=false
      bMiteredCorners=false
      bSeparateFillAlpha=false
      outlineMaterial=null
      outlineSize=0
      outlineColor=<rgba 0 0 0 1>
     }
    }
    shadowOffset=<0 0>
   }
   slot {class=VerticalBoxSlot, properties {padding=<ltrb 0 0 0 15>}}
  }
  HorizontalBox_64 {widgetClassPath=&HorizontalBox, parent=VB_RepLevel, widget=HorizontalBox_64, slot {class=VerticalBoxSlot}}
  Progress {
   widgetClassPath=&Image
   bIsVariable=true
   parent=HorizontalBox_64
   widget=Progress
   properties {brush {
     bIsDynamicallyLoaded=false
     drawAs=Image
     imageType=FullColor
     mirroring=NoMirror
     resourceName=None
     resourceObject=/FortniteCoreUI/UI/Quests/MaterialInstances/MI_UI_QuestProgressBar
     tiling=NoTile
     imageSize=<32 10>
     margin=<ltrb 0.5 0.5 0.5 0.5>
     outlineSettings {
      bUseBrushTransparency=false
      roundingType=HalfHeightRadius
      width=0
      color {colorUseRule=UseColor_Specified, specifiedColor=<rgba 0 0 0 0>}
      cornerRadii=<0 0 0 0>
     }
     tintColor {colorUseRule=UseColor_Specified, specifiedColor=<rgba 1 1 1 1>}
     uVRegion {bIsValid=false, max=<0 0>, min=<0 0>}
    }}
   slot {class=HorizontalBoxSlot, properties {verticalAlignment=VAlign_Center, padding=<ltrb 0 0 8 0>, size {sizeRule=Fill, value=1}}}
  }
  RichText_Progress {
   widgetClassPath=&CommonRichTextBlock
   bIsVariable=true
   parent=HorizontalBox_64
   widget=RichText_Progress
   properties {
    bOverrideDefaultStyle=true
    text=99 / 99
    defaultTextStyleOverride {
     overflowPolicy=Clip
     transformPolicy=None
     colorAndOpacity {colorUseRule=UseColor_Specified, specifiedColor=<rgba 1 1 1 1>}
     font {
      bForceMonospaced=false
      bMaterialIsStencil=false
      fontMaterial=null
      fontObject=&HeadingNow
      letterSpacing=0
      monospacedWidth=1
      size=15.75
      skewAmount=0
      typefaceFontName=75 Medium
      outlineSettings {
       bApplyOutlineToDropShadows=false
       bMiteredCorners=false
       bSeparateFillAlpha=false
       outlineMaterial=null
       outlineSize=0
       outlineColor=<rgba 0 0 0 1>
      }
     }
     highlightColor {colorUseRule=UseColor_Specified, specifiedColor=<rgba 0 0 0 1>}
     highlightShape {
      bIsDynamicallyLoaded=false
      drawAs=Image
      imageType=NoImage
      mirroring=NoMirror
      resourceName=None
      resourceObject=null
      tiling=NoTile
      imageSize=<32 32>
      margin=<ltrb 0 0 0 0>
      outlineSettings {
       bUseBrushTransparency=false
       roundingType=HalfHeightRadius
       width=0
       color {colorUseRule=UseColor_Specified, specifiedColor=<rgba 0 0 0 0>}
       cornerRadii=<0 0 0 0>
      }
      tintColor {colorUseRule=UseColor_Specified, specifiedColor=<rgba 1 1 1 1>}
      uVRegion {bIsValid=false, max=<0 0>, min=<0 0>}
     }
     selectedBackgroundColor {colorUseRule=UseColor_Foreground, specifiedColor=<rgba 1 0 1 1>}
     shadowColorAndOpacity=<rgba 0 0 0 1>
     shadowOffset=<0 0>
     strikeBrush {
      bIsDynamicallyLoaded=false
      drawAs=Image
      imageType=NoImage
      mirroring=NoMirror
      resourceName=None
      resourceObject=null
      tiling=NoTile
      imageSize=<32 32>
      margin=<ltrb 0 0 0 0>
      outlineSettings {
       bUseBrushTransparency=false
       roundingType=HalfHeightRadius
       width=0
       color {colorUseRule=UseColor_Specified, specifiedColor=<rgba 0 0 0 0>}
       cornerRadii=<0 0 0 0>
      }
      tintColor {colorUseRule=UseColor_Specified, specifiedColor=<rgba 1 1 1 1>}
      uVRegion {bIsValid=false, max=<0 0>, min=<0 0>}
     }
     underlineBrush {
      bIsDynamicallyLoaded=false
      drawAs=Image
      imageType=NoImage
      mirroring=NoMirror
      resourceName=None
      resourceObject=null
      tiling=NoTile
      imageSize=<32 32>
      margin=<ltrb 0 0 0 0>
      outlineSettings {
       bUseBrushTransparency=false
       roundingType=HalfHeightRadius
       width=0
       color {colorUseRule=UseColor_Specified, specifiedColor=<rgba 0 0 0 0>}
       cornerRadii=<0 0 0 0>
      }
      tintColor {colorUseRule=UseColor_Specified, specifiedColor=<rgba 1 1 1 1>}
      uVRegion {bIsValid=false, max=<0 0>, min=<0 0>}
     }
    }
   }
   slot {class=HorizontalBoxSlot}
  }
  EntryBox_ServiceGroups {
   widgetClassPath=&DynamicEntryBox
   bInherited=true
   parent=VerticalBox_17
   widget=EntryBox_ServiceGroups
   properties {
    entryBoxType=Vertical
    entryWidgetClass=&/CollectionNPCServiceGroup.CollectionNPCServiceGroup_C
    numDesignerPreviewEntries=2
    entrySpacing=<0 8>
   }
   slot {class=VerticalBoxSlot, properties {padding=<ltrb 0 8 0 0>, size {sizeRule=Fill, value=1}}}
  }
 ]
}
