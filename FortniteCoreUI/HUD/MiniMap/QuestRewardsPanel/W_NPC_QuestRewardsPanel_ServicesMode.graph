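# NPC quest rewards panel in services mode for the minimap HUD, showing a character portrait over glows, the name, inherited service list and description.
# Show sets the name from the NPC title or an override, and the description via Format Text without placeholder arguments.
# The portrait lazy-loads into a dynamic material, with ClippingMask zeroed on PreConstruct and after each brush change, and Disable512Scaling enabled on PreConstruct.
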WidgetBlueprint /FortniteCoreUI/HUD/MiniMap/QuestRewardsPanel/W_NPC_QuestRewardsPanel_ServicesMode
graphs [EventGraph {nodes [
   Event PreConstruct()
   > n1 = GetDynamicMaterial(self=Get LazyImg_Character() -> (LazyImg_Character:Fort Lazy Image Object Reference))
   > SetScalarParameterValue(self=@n1, ParameterName=ClippingMask, Value=0)
   > SetScalarParameterValue(self=@n1, ParameterName=Disable512Scaling, Value=1)
   n2 = Event Show!() -> (OutputDelegate, then, NPCTitle:"Text (by ref)", NPCPicture:"Soft Object Reference (by ref)") {is=title K2Node_Event}
   > n3 = Sequence!(execute) -> (then_0, then_1, then_2:Exec)
   > n4 = Branch(Condition=Get Override NPC Name() -> (Override NPC Name:bool))
   > Set Text(
    Text=Get NPCName_Override() -> (NPCName_Override:Text)
    self:Rich Text Block Object Reference=Get RichTextBlock_NPCName() -> (RichTextBlock_NPCName:Common Rich Text Block Object Reference)
   )
   Set Text(execute=@n4.else, Text=@n2.NPCTitle, self=Get RichTextBlock_NPCName())
   Set Text(
    execute=@n3.then_1
    Text=Format Text(Format=Get PanelDescription() -> (PanelDescription:Text))
    self=Get RichText_Description() -> (RichText_Description:Common Rich Text Block Object Reference)
   )
   n5 = Get LazyImg_Character()
   SetBrushFromLazyDisplayAsset!(
    execute=@n3.then_2
    self:Common Lazy Image Object Reference=@n5
    LazyObject:"Soft Object Reference (by ref)"=@n2.NPCPicture
    bMatchTextureSize:bool=false
   ) -> (then)
   > n6 = GetDynamicMaterial(self=@n5)
   > SetScalarParameterValue(self=@n6, ParameterName=ClippingMask, Value=0)
  ]}]
variables [
 NPCName_Override {type=string, default=""}
 Override NPC Name {type=boolean}
 PanelDescription {
  type=string
  default="Play through {0}'s Questline in Battle Royale or Zero Build to increase your Weapon Expertise with {1} and acquire services in-match!"
 }
]
widgets {
 info {inheritedWidgetCount=1, parentClass=/Script/FortniteUI.AthenaNPCQuestServicesPanel, rootWidgetClass=&SizeBox, widgetCount=10}
 widgets [
  SizeBox_0 {widgetClassPath=&SizeBox, widget=SizeBox_0, properties {bOverride_WidthOverride=true, heightOverride=940, widthOverride=847}}
  Container {widgetClassPath=&Overlay, bIsVariable=true, parent=SizeBox_0, widget=Container, slot {class=SizeBoxSlot}}
  RichTextBlock_NPCName {
   widgetClassPath=&CommonRichTextBlock
   bIsVariable=true
   parent=Container
   widget=RichTextBlock_NPCName
   properties {
    bOverrideDefaultStyle=true
    justification=Center
    lineHeightPercentage=0.9
    scrollStyle=&TextScrollStyle-Base_C
    text=NPC NAME
    textTransformPolicy=ToUpper
    defaultTextStyleOverride {
     overflowPolicy=Clip
     transformPolicy=None
     colorAndOpacity {colorUseRule=UseColor_Specified, specifiedColor=<rgba 1 1 1 1>}
     font {
      bForceMonospaced=false
      bMaterialIsStencil=false
      fontMaterial=null
      fontObject=&Burbank
      letterSpacing=0
      monospacedWidth=1
      size=54
      skewAmount=0.1
      typefaceFontName=BurbankRegular-Black
      outlineSettings {
       bApplyOutlineToDropShadows=false
       bMiteredCorners=false
       bSeparateFillAlpha=false
       outlineMaterial=null
       outlineSize=0
       outlineColor=<rgba 0 0 0 1>
      }
     }
     highlightColor {colorUseRule=UseColor_Specified, specifiedColor=<rgba 0 0 0 1>}
     highlightShape {
      bIsDynamicallyLoaded=false
      drawAs=Image
      imageType=NoImage
      mirroring=NoMirror
      resourceName=None
      resourceObject=null
      tiling=NoTile
      imageSize=<32 32>
      margin=<ltrb 0 0 0 0>
      outlineSettings {
       bUseBrushTransparency=false
       roundingType=HalfHeightRadius
       width=0
       color {colorUseRule=UseColor_Specified, specifiedColor=<rgba 0 0 0 0>}
       cornerRadii=<0 0 0 0>
      }
      tintColor {colorUseRule=UseColor_Specified, specifiedColor=<rgba 1 1 1 1>}
      uVRegion {bIsValid=false, max=<0 0>, min=<0 0>}
     }
     selectedBackgroundColor {colorUseRule=UseColor_Foreground, specifiedColor=<rgba 1 0 1 1>}
     shadowColorAndOpacity=<rgba 0.0549 0.0745 0.3294 1>
     shadowOffset=<2 2>
     strikeBrush {
      bIsDynamicallyLoaded=false
      drawAs=NoDrawType
      imageType=NoImage
      mirroring=NoMirror
      resourceName=None
      resourceObject=null
      tiling=NoTile
      imageSize=<32 32>
      margin=<ltrb 0 0 0 0>
      outlineSettings {
       bUseBrushTransparency=false
       roundingType=HalfHeightRadius
       width=0
       color {colorUseRule=UseColor_Specified, specifiedColor=<rgba 0 0 0 0>}
       cornerRadii=<0 0 0 0>
      }
      tintColor {colorUseRule=UseColor_Specified, specifiedColor=<rgba 1 1 1 1>}
      uVRegion {bIsValid=false, max=<0 0>, min=<0 0>}
     }
     underlineBrush {
      bIsDynamicallyLoaded=false
      drawAs=Image
      imageType=NoImage
      mirroring=NoMirror
      resourceName=None
      resourceObject=null
      tiling=NoTile
      imageSize=<32 32>
      margin=<ltrb 0 0 0 0>
      outlineSettings {
       bUseBrushTransparency=false
       roundingType=HalfHeightRadius
       width=0
       color {colorUseRule=UseColor_Specified, specifiedColor=<rgba 0 0 0 0>}
       cornerRadii=<0 0 0 0>
      }
      tintColor {colorUseRule=UseColor_Specified, specifiedColor=<rgba 1 1 1 1>}
      uVRegion {bIsValid=false, max=<0 0>, min=<0 0>}
     }
    }
    margin=<ltrb 0 0 0 16>
   }
   slot {class=OverlaySlot, properties {horizontalAlignment=HAlign_Center, padding=<ltrb 0 80 0 0>}}
  }
  Image_BackgroundGlow {
   widgetClassPath=&Image
   bIsVariable=true
   parent=Container
   widget=Image_BackgroundGlow
   properties {
    brush {
     bIsDynamicallyLoaded=false
     drawAs=Image
     imageType=NoImage
     mirroring=NoMirror
     resourceName=None
     resourceObject=/FortniteCoreUI/HUD/MiniMap/QuestRewardsPanel/Materials/MI_UI_NPCQuestRewards_Glow_BG
     tiling=NoTile
     imageSize=<1025 1025>
     margin=<ltrb 0 0 0 0>
     outlineSettings {
      bUseBrushTransparency=false
      roundingType=HalfHeightRadius
      width=0
      color {colorUseRule=UseColor_Specified, specifiedColor=<rgba 0 0 0 0>}
      cornerRadii=<0 0 0 0>
     }
     tintColor {colorUseRule=UseColor_Specified, specifiedColor=<rgba 1 1 1 1>}
     uVRegion {bIsValid=false, max=<0 0>, min=<0 0>}
    }
    colorAndOpacity=<rgba 1 1 1 0.6714>
    renderTransform {angle=0, scale=<1.3 1.3>, shear=<0 0>, translation=<0 -100>}
   }
   slot {class=OverlaySlot, properties {horizontalAlignment=HAlign_Center, verticalAlignment=VAlign_Center}}
  }
  Image_ForegroundGlow {
   widgetClassPath=&Image
   bIsVariable=true
   parent=Container
   widget=Image_ForegroundGlow
   properties {
    brush {
     bIsDynamicallyLoaded=false
     drawAs=Image
     imageType=NoImage
     mirroring=NoMirror
     resourceName=None
     resourceObject=&MI_UI_NPCQuestRewards_Glow_BG
     tiling=NoTile
     imageSize=<1024 1024>
     margin=<ltrb 0 0 0 0>
     outlineSettings {
      bUseBrushTransparency=false
      roundingType=HalfHeightRadius
      width=0
      color {colorUseRule=UseColor_Specified, specifiedColor=<rgba 0 0 0 0>}
      cornerRadii=<0 0 0 0>
     }
     tintColor {colorUseRule=UseColor_Specified, specifiedColor=<rgba 1 1 1 1>}
     uVRegion {bIsValid=false, max=<0 0>, min=<0 0>}
    }
    colorAndOpacity=<rgba 1 1 1 0.0417>
    renderTransform {angle=0, scale=<1.25 1.25>, shear=<0 0>, translation=<0 -83>}
   }
   slot {class=OverlaySlot, properties {horizontalAlignment=HAlign_Right, verticalAlignment=VAlign_Center, padding=<ltrb 0 0 -50 -200>}}
  }
  LazyImg_Character {
   widgetClassPath=&FortLazyImage
   bIsVariable=true
   parent=Container
   widget=LazyImg_Character
   properties {
    materialTextureParamName=ItemIcon
    brush {
     bIsDynamicallyLoaded=false
     drawAs=Image
     imageType=NoImage
     mirroring=NoMirror
     resourceName=None
     resourceObject=/FortniteCoreUI/HUD/MiniMap/QuestRewardsPanel/Materials/MI_UI_Character_Picture
     tiling=NoTile
     imageSize=<500 500>
     margin=<ltrb 0 0 0 0>
     outlineSettings {
      bUseBrushTransparency=false
      roundingType=HalfHeightRadius
      width=0
      color {colorUseRule=UseColor_Specified, specifiedColor=<rgba 0 0 0 0>}
      cornerRadii=<0 0 0 0>
     }
     tintColor {colorUseRule=UseColor_Specified, specifiedColor=<rgba 1 1 1 1>}
     uVRegion {bIsValid=false, max=<0 0>, min=<0 0>}
    }
    renderTransform {angle=0, scale=<1.7 1.7>, shear=<0 0>, translation=<6 100>}
   }
   slot {class=OverlaySlot, properties {horizontalAlignment=HAlign_Center, verticalAlignment=VAlign_Center}}
  }
  VerticalBox_0 {
   widgetClassPath=&VerticalBox
   parent=Container
   widget=VerticalBox_0
   slot {class=OverlaySlot, properties {horizontalAlignment=HAlign_Center, verticalAlignment=VAlign_Bottom, padding=<ltrb 0 0 0 30>}}
  }
  Services {
   widgetClassPath=/FortniteCoreUI/HUD/MiniMap/Services/CollectionNPCServiceContainer
   bInherited=true
   bIsVariable=true
   parent=VerticalBox_0
   widget=Services
   properties {bExcludeAlwaysAvailableServices=true, bShowLinesBetweenServices=true, enableHeader=false, maxGroupSize=3}
   slot {class=VerticalBoxSlot, properties {horizontalAlignment=HAlign_Center, verticalAlignment=VAlign_Center, padding=<ltrb 0 0 0 50>}}
  }
  Border_53 {
   widgetClassPath=&Border
   parent=VerticalBox_0
   widget=Border_53
   properties {
    background {
     bIsDynamicallyLoaded=false
     drawAs=Image
     imageType=NoImage
     mirroring=NoMirror
     resourceName=None
     resourceObject=/FortniteCoreUI/HUD/MiniMap/QuestRewardsPanel/Materials/MI_UI_QR_InfoBackground
     tiling=NoTile
     imageSize=<32 32>
     margin=<ltrb 0 0 0 0>
     outlineSettings {
      bUseBrushTransparency=false
      roundingType=HalfHeightRadius
      width=0
      color {colorUseRule=UseColor_Specified, specifiedColor=<rgba 0 0 0 0>}
      cornerRadii=<0 0 0 0>
     }
     tintColor {colorUseRule=UseColor_Specified, specifiedColor=<rgba 1 1 1 1>}
     uVRegion {bIsValid=false, max=<0 0>, min=<0 0>}
    }
    padding=<ltrb 50 20 50 20>
   }
   slot {class=VerticalBoxSlot}
  }
  RichText_Description {
   widgetClassPath=&CommonRichTextBlock
   bIsVariable=true
   parent=Border_53
   widget=RichText_Description
   properties {
    autoWrapText=true
    justification=Center
    text=Description
    textStyleSet=/FortniteCoreUI/HUD/MiniMap/QuestRewardsPanel/Materials/DT_UI_RichTextStyles_QuestRewards
    wrapTextAt=800
    decoratorClasses [/Game/UI/Sidebar/RebootRally/BP_UI_RebootRallyRichImageDecorator.BP_UI_RebootRallyRichImageDecorator_C]
    defaultTextStyleOverride {
     overflowPolicy=Clip
     transformPolicy=None
     colorAndOpacity {colorUseRule=UseColor_Specified, specifiedColor=<rgba 1 1 1 1>}
     font {
      bForceMonospaced=false
      bMaterialIsStencil=false
      fontMaterial=null
      fontObject=&BurbankSmall
      letterSpacing=0
      monospacedWidth=1
      size=19
      skewAmount=0
      typefaceFontName=Bold
      outlineSettings {
       bApplyOutlineToDropShadows=false
       bMiteredCorners=false
       bSeparateFillAlpha=false
       outlineMaterial=null
       outlineSize=0
       outlineColor=<rgba 0 0 0 1>
      }
     }
     highlightColor {colorUseRule=UseColor_Specified, specifiedColor=<rgba 0 0 0 1>}
     highlightShape {
      bIsDynamicallyLoaded=false
      drawAs=Image
      imageType=NoImage
      mirroring=NoMirror
      resourceName=None
      resourceObject=null
      tiling=NoTile
      imageSize=<32 32>
      margin=<ltrb 0 0 0 0>
      outlineSettings {
       bUseBrushTransparency=false
       roundingType=HalfHeightRadius
       width=0
       color {colorUseRule=UseColor_Specified, specifiedColor=<rgba 0 0 0 0>}
       cornerRadii=<0 0 0 0>
      }
      tintColor {colorUseRule=UseColor_Specified, specifiedColor=<rgba 1 1 1 1>}
      uVRegion {bIsValid=false, max=<0 0>, min=<0 0>}
     }
     selectedBackgroundColor {colorUseRule=UseColor_Foreground, specifiedColor=<rgba 1 0 1 1>}
     shadowColorAndOpacity=<rgba 0 0 0 1>
     shadowOffset=<0 0>
     strikeBrush {
      bIsDynamicallyLoaded=false
      drawAs=Image
      imageType=NoImage
      mirroring=NoMirror
      resourceName=None
      resourceObject=null
      tiling=NoTile
      imageSize=<32 32>
      margin=<ltrb 0 0 0 0>
      outlineSettings {
       bUseBrushTransparency=false
       roundingType=HalfHeightRadius
       width=0
       color {colorUseRule=UseColor_Specified, specifiedColor=<rgba 0 0 0 0>}
       cornerRadii=<0 0 0 0>
      }
      tintColor {colorUseRule=UseColor_Specified, specifiedColor=<rgba 1 1 1 1>}
      uVRegion {bIsValid=false, max=<0 0>, min=<0 0>}
     }
     underlineBrush {
      bIsDynamicallyLoaded=false
      drawAs=Image
      imageType=NoImage
      mirroring=NoMirror
      resourceName=None
      resourceObject=null
      tiling=NoTile
      imageSize=<32 32>
      margin=<ltrb 0 0 0 0>
      outlineSettings {
       bUseBrushTransparency=false
       roundingType=HalfHeightRadius
       width=0
       color {colorUseRule=UseColor_Specified, specifiedColor=<rgba 0 0 0 0>}
       cornerRadii=<0 0 0 0>
      }
      tintColor {colorUseRule=UseColor_Specified, specifiedColor=<rgba 1 1 1 1>}
      uVRegion {bIsValid=false, max=<0 0>, min=<0 0>}
     }
    }
   }
   slot {class=BorderSlot, properties {padding=<ltrb 50 20 50 20>}}
  }
 ]
}
